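; Looping gameplay cue base for an active BuckSixteen buff that keeps its idle particle effect in sync with the player's water state.
; On loop start it caches the first particle component and target player pawn, then sets the effect's WaterFade parameter from surface-swimming state.
; It binds the pawn's enter and exit water events to switch WaterFade on or off without ticking, and unbinds them on removal.
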(event EventBeginPlay)

(event Collision|EventActorBeginOverlap
  (params (OtherActor :type Actor%20Object%20Reference)))

(event EventTick
  (params (DeltaSeconds :type Float%20%28single-precision%29)))

(event EventOnLoopingStart(EitherVFX-preferthisversion)
  (params (MyTarget :type Actor%20Object%20Reference) (Parameters :type Gameplay%20Cue%20Parameters%20Structure%20%28by%20ref%29) (ParticleComponents :type Array%20of%20FXSystem%20Component%20Object%20References) (AudioComponents :type Array%20of%20Audio%20Component%20Object%20References))
  (bind _buffidlefx (Variables|Default|GetBuffIdleFX))
  (bind _player_pawn (Variables|Default|GetPlayer%20Pawn))
  (|CallParentFunction MyTarget Parameters ParticleComponents AudioComponents)
  (Variables|Default|SetBuffIdleFX (Utilities|Array|Get(acopy) ParticleComponents))
  (bind _asfort_player_pawn (Utilities|Casting|CastToFortPlayerPawn MyTarget))
  (bind _output_get (Variables|Default|SetPlayer%20Pawn _asfort_player_pawn))
  (Utilities|FlowControl|Sequence
    (:then_0
      ;[1 0.7866 0.1638 0.5] If player is Swimming or not when spawning FX
      (if (WaterBody|IsSurfaceSwimming _output_get)
        (Effects|Components|ParticleSystem|SetFloatParameter _buffidlefx "WaterFade" 1.0)
        (else
          (Effects|Components|ParticleSystem|SetFloatParameter _buffidlefx "WaterFade" 0.0))))
    (:then_1
      ;[0.2345 1 0.4222 0.5] Bind to Player Water Events
      (Utilities|FlowControl|Sequence
        (:then_0
          (EventDispatchers|BindEventtoOnPawnEnterWater _player_pawn (EventDispatchers|CreateEvent :!event Custom|EnterWaterFXFade)))
        (:then_1
          (EventDispatchers|BindEventtoOnPawnExitWater _player_pawn (EventDispatchers|CreateEvent :!event Custom|ExitWaterFXFade)))))))

(event EventOnRemoval(EitherVFX-preferthisversion)
  (params (MyTarget :type Actor%20Object%20Reference) (Parameters :type Gameplay%20Cue%20Parameters%20Structure%20%28by%20ref%29) (ParticleComponents :type Array%20of%20FXSystem%20Component%20Object%20References) (AudioComponents :type Array%20of%20Audio%20Component%20Object%20References) (BurstCameraShakeInstance :type Camera%20Shake%20Base%20Object%20Reference) (BurstDecalInstance :type Decal%20Actor%20Object%20Reference))
  (bind _player_pawn (Variables|Default|GetPlayer%20Pawn))
  (|CallParentFunction MyTarget Parameters ParticleComponents AudioComponents BurstCameraShakeInstance BurstDecalInstance)
  ;[1 0.1983 0.2294 0.5] Unbind from Player Water Events
  (Utilities|FlowControl|Sequence
    (:then_0
      (EventDispatchers|UnbindEventfromOnPawnEnterWater _player_pawn (EventDispatchers|CreateEvent :!event Custom|EnterWaterFXFade)))
    (:then_1
      (EventDispatchers|UnbindEventfromOnPawnExitWater _player_pawn (EventDispatchers|CreateEvent :!event Custom|ExitWaterFXFade)))))

(event Custom|EnterWaterFXFade
  (params (FortPlayerPawn :type Fort%20Player%20Pawn%20Object%20Reference))
  ;[0.233 0.4946 0.8594 0.5] FX Water Fade In & Out
  (Effects|Components|ParticleSystem|SetFloatParameter (Variables|Default|GetBuffIdleFX) "WaterFade" 1.0))

(event Custom|ExitWaterFXFade
  (params (FortPlayerPawn :type Fort%20Player%20Pawn%20Object%20Reference) (bPawnIsOutsideOfAllWaterBodies :type Boolean))
  ;[0.233 0.4946 0.8594 0.5] FX Water Fade In & Out
  (Effects|Components|ParticleSystem|SetFloatParameter (Variables|Default|GetBuffIdleFX) "WaterFade" 0.0))

(fn ConstructionScript)
